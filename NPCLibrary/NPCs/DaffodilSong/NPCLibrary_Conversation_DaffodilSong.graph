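# Conversation database for the DaffodilSong NPC, who greets players in character as Apollo, the overworked god of the sun.
# The entry point leads to one Speech node with a default line and contextual lines gated by first-ever meeting, per-match conversation count or specific player outfits.
# Curve-row priorities are set for the first-meeting and outfit lines.
# The speech then branches through link nodes to shared quests, basic services and this NPC's Dialogue entry.
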ConversationDatabase /NPCLibrary/NPCs/DaffodilSong/NPCLibrary_Conversation_DaffodilSong
entry_points [{tag=Conversation.Entry.DaffodilSong, destinations [#1]}]
nodes [
 ConversationEntryPointNode_7 {
  id=#1
  class=ConversationEntryPointNode
  outputs [#2]
  properties {entryTag {tagName=Conversation.Entry.DaffodilSong}}
 }
 Speech_C_4 {
  id=#2
  class=Speech_C
  outputs [#3, #4, #5]
  properties {
   nodeName=""
   generalConfig {
    defaultMessage="I am Apollo, god of the sun... and a FEW other things."
    contextualMessages [
     {
      message="Greetings, mortal."
      contextRequirements [{requirement {class=FortControllerRequirement_IsFirstEverConversationWithNPC}}]
      priority {curve {rowName=Tandem.Speech.Priority.FirstEverInteraction}}
     }
     {
      message="I am Apollo, god of the sun... and a FEW other things."
      contextRequirements [{requirement {class=FortControllerRequirement_CheckNPCMatchConversationCount, properties {nPCUniqueIdTag {tagName=AISpawnerData.Type.Tandem.DaffodilSong}}}}]
      priority {curve {rowName=None}}
     }
     {
      message="My To Do list: direct the muses, consult the oracles, heal the wounded, keep the sun in the sky..."
      contextRequirements [{requirement {class=FortControllerRequirement_CheckNPCMatchConversationCount, properties {nPCUniqueIdTag {tagName=AISpawnerData.Type.Tandem.DaffodilSong}}}}]
      priority {curve {rowName=None}}
     }
     {
      message=Of course Ares has the free time to scrap with every mortal who challenges him. *sigh*
      contextRequirements [{requirement {class=FortControllerRequirement_CheckNPCMatchConversationCount, properties {nPCUniqueIdTag {tagName=AISpawnerData.Type.Tandem.DaffodilSong}}}}]
      priority {curve {rowName=None}}
     }
     {
      message=Where does Zeus have TIME for a vendetta against mortalkind??
      contextRequirements [{requirement {class=FortControllerRequirement_CheckNPCMatchConversationCount, properties {nPCUniqueIdTag {tagName=AISpawnerData.Type.Tandem.DaffodilSong}}}}]
      priority {curve {rowName=None}}
     }
     {
      message="At my request, the muses inspired the invention of the spreadsheet. How would I function without it?"
      contextRequirements [{requirement {class=FortControllerRequirement_CheckNPCMatchConversationCount, properties {nPCUniqueIdTag {tagName=AISpawnerData.Type.Tandem.DaffodilSong}}}}]
      priority {curve {rowName=None}}
     }
     {
      message="If you appreciate the sun, music, dance, healing, prophecies, etc... I wouldn't mind a thank you."
      contextRequirements [{requirement {class=FortControllerRequirement_CheckNPCMatchConversationCount, properties {nPCUniqueIdTag {tagName=AISpawnerData.Type.Tandem.DaffodilSong}}}}]
      priority {curve {rowName=None}}
     }
     {
      message=You should meet my Oracle at her scrying pools. She's very fun.
      contextRequirements [{requirement {class=FortControllerRequirement_CheckNPCMatchConversationCount, properties {nPCUniqueIdTag {tagName=AISpawnerData.Type.Tandem.DaffodilSong}}}}]
      priority {curve {rowName=None}}
     }
     {
      message=Sister. How fares the Eternal Hunt?
      contextRequirements [{requirement {
         class=FortControllerRequirement_HasCID
         properties {allowedCIds [/BRCosmetics/Athena/Items/Cosmetics/Characters/Character_SeleneCobra_Stride]}
        }}]
      priority {curve {rowName=Tandem.Speech.Priority.OutfitReaction}}
     }
     {
      message=I can't imagine being the god of only one thing. It must be terribly boring.
      contextRequirements [{requirement {
         class=FortControllerRequirement_HasCID
         properties {allowedCIds [/BRCosmetics/Athena/Items/Cosmetics/Characters/Character_GoldAccomplishment]}
        }}]
      priority {curve {rowName=Tandem.Speech.Priority.OutfitReaction}}
     }
     {
      message=You have my support. The mortals deserve protection.
      contextRequirements [{requirement {class=FortControllerRequirement_HasCID, properties {allowedCIds [/BRCosmetics/Athena/Items/Cosmetics/Characters/Character_LiftingAura_Wave]}}}]
      priority {curve {rowName=Tandem.Speech.Priority.OutfitReaction}}
     }
    ]
   }
  }
 }
 ConversationLinkNode_2 {
  id=#5
  class=ConversationLinkNode
  properties {remoteEntryTag {tagName=Conversation.Entry.DaffodilSong.Dialogue}}
 }
 ConversationLinkNode_0 {id=#4, class=ConversationLinkNode, properties {remoteEntryTag {tagName=Conversation.Entry.BasicServices}}}
 ConversationLinkNode_1 {id=#3, class=ConversationLinkNode, properties {remoteEntryTag {tagName=Conversation.Entry.NPCLibrary.Quests}}}
]
