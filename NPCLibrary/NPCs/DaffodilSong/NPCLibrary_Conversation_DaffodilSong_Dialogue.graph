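# Conversation database for the NPC DaffodilSong, providing an "Ask about..." lore dialogue through a single entry tag.
# The opening question branches into a Zeus topic, a topic on the NPC's chosen Oracle, and a Back option.
# Each topic plays a linear chain of three speech lines, and both chains converge on a shared ReturnToStart node with a BACK choice.
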ConversationDatabase /NPCLibrary/NPCs/DaffodilSong/NPCLibrary_Conversation_DaffodilSong_Dialogue
entry_points [{tag=Conversation.Entry.DaffodilSong.Dialogue, destinations [#1]}]
nodes [
 ConversationEntryPointNode_0 {
  id=#1
  class=ConversationEntryPointNode
  outputs [#2]
  properties {entryTag {tagName=Conversation.Entry.DaffodilSong.Dialogue}}
 }
 Speech_C_0 {
  id=#2
  class=Speech_C
  outputs [#3, #4, #5]
  properties {nodeName="", softTaskIcon=&T_UI_NPC_question, generalConfig {defaultMessage="What would you like to know, hm?"}}
  sub_nodes [ConversationChoiceNode_5 {class=ConversationChoiceNode, properties {defaultChoiceDisplayText=Ask about...}}]
 }
 Back_C_0 {
  id=#5
  class=Back_C
  sub_nodes [ConversationChoiceNode_7 {class=ConversationChoiceNode, properties {defaultChoiceDisplayText=BACK}}]
 }
 Speech_C_6 {
  id=#4
  class=Speech_C
  outputs [#6]
  properties {
   nodeName=""
   softTaskIcon=&T_UI_NPC_question
   generalConfig {defaultMessage="Ah, yes! My chosen Oracle. Odd duck, that one."}
  }
  sub_nodes [ConversationChoiceNode_10 {class=ConversationChoiceNode, properties {defaultChoiceDisplayText=The Oracle}}]
 }
 Speech_C_12 {
  id=#6
  class=Speech_C
  outputs [#7]
  properties {
   nodeName=""
   softTaskIcon=&T_UI_NPC_question
   generalConfig {defaultMessage="She has a natural talent for seeing things clearly, and a strange affection for you specifically. You can trust her."}
  }
 }
 Speech_C_13 {
  id=#7
  class=Speech_C
  outputs [#8]
  properties {
   nodeName=""
   softTaskIcon=&T_UI_NPC_question
   generalConfig {defaultMessage="She was a mortal once too, you know! A shame, that. Nobody's perfect, I suppose!"}
  }
 }
 ReturnToStart_C_23 {
  id=#8
  class=ReturnToStart_C
  properties {softTaskIcon=&T_UI_NPC_Back}
  sub_nodes [ConversationChoiceNode_32 {class=ConversationChoiceNode, properties {defaultChoiceDisplayText=BACK}}]
 }
 Speech_C_7 {
  id=#3
  class=Speech_C
  outputs [#9]
  properties {
   nodeName=""
   softTaskIcon=&T_UI_NPC_question
   generalConfig {defaultMessage=I try to stay out of the petty squabbles of my immortal family.}
  }
  sub_nodes [ConversationChoiceNode_15 {class=ConversationChoiceNode, properties {defaultChoiceDisplayText=Zeus}}]
 }
 Speech_C_15 {
  id=#9
  class=Speech_C
  outputs [#10]
  properties {
   nodeName=""
   softTaskIcon=&T_UI_NPC_question
   generalConfig {defaultMessage="Most think Zeus has long tired of mortalkind's hubris, and your Society pushed him over the edge."}
  }
 }
 Speech_C_14 {
  id=#10
  class=Speech_C
  outputs [#8]
  properties {
   nodeName=""
   softTaskIcon=&T_UI_NPC_question
   generalConfig {defaultMessage="But I see all prophecies. I know his reign is destined to end, by a mortal's hand."}
  }
 }
]
